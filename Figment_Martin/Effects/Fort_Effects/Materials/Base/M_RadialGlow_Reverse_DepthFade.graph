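# Unlit additive particle sprite material that draws a soft radial glow with a reversed depth fade, so it shows where it nears scene geometry.
# Particle color times a radial gradient feeds the shared FX emissive adjustments function.
# Particle alpha times the same gradient is masked by clamped inverted DepthFade, with FadeDistance set through a dynamic parameter.
# A feature level switch replaces the depth mask with a constant on ES2 and ES3_1.
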Material /Figment_Martin/Effects/Fort_Effects/Materials/Base/M_RadialGlow_Reverse_DepthFade
graph {
 properties {bUsedWithParticleSprites=true, blendMode=BLEND_Additive, pixelDepthOffsetMode=PDOM_Legacy, shadingModel=MSM_Unlit, usedShadingModels=Unlit}
 nodes [
  n1 = ParticleColor()
  n2 = MaterialFunctionCall!(Radius, CenterPosition) {materialFunction=/Engine/Functions/Engine_MaterialFunctions01/Gradient/RadialGradient}
  MaterialFunctionCall(Emissive=Multiply(A=@n1.RGB, B=@n2.RadialGradient)) {materialFunction=&MF_FX_EmissiveAdjustments}
  n3 = DynamicParameter() {defaultValue=<rgba 16 0 0 1>, paramNames [FadeDistance,Param2,Param3,Param4]}
  n4 = Clamp(None=OneMinus(None=DepthFade(FadeDistance=@n3.FadeDistance)))
  n5 = Constant() {r=1}
  Multiply(A=Multiply(A=@n1.A, B=@n2.RadialGradient), B=FeatureLevelSwitch(Default=@n4, ES2=@n5, ES3_1=@n5, SM4_REMOVED=@n4, SM5=@n4, SM6=@n4))
 ]
}
